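# Subsurface material for LanternTown paper lanterns, usable on Niagara mesh particles, with diffuse, mask and normal textures.
# Particle color tints both the diffuse texture and a glow term shaped by the mask's green channel raised to a power.
# The glow flickers by lerping between a slightly dimmed and full warm color using two summed time-driven sines, both phase-offset by the Rand dynamic parameter.
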Material /Game/Effects/Fort_Effects/Effects/Environment/LanternTown/Meshes/MI_LightSource_PaperLanterns_A
graph {
 properties {
  assetImportData="<self>:InterchangeAssetImportData_0"
  bUsedWithNiagaraMeshParticles=true
  shadingModel=MSM_Subsurface
  usedShadingModels=Subsurface
 }
 nodes [
  n1 = ParticleColor()
  Multiply(A=TextureSample() {texture=&/T_LightSource_PaperLantern_A_D}.RGB, B=@n1.RGB)
  n2 = TextureSample() {samplerType=SAMPLERTYPE_Masks, texture=&/T_LightSource_PaperLantern_A_M}
  n3 = ScalarParameter() {defaultValue=2}
  n4 = VectorParameter() {defaultValue=<rgba 2 0.9042 0 1>}
  n5 = ScalarParameter() {defaultValue=0.955, parameterName=Param_2}
  n6 = Multiply(A=Time(), B=ScalarParameter() {defaultValue=5, parameterName=Flicker 1})
  n7 = DynamicParameter() {paramNames [Rand,Param2,Param3,Param4]}
  n8 = Multiply(A=Sine(None=Add(A=@n6, B=@n7.Rand)), B=ScalarParameter() {defaultValue=2, parameterName=Param_3})
  n9 = Time()
  n10 = ScalarParameter() {defaultValue=5, parameterName=Random Time Flicker 2}
  n11 = Sine(None=Add(A=Multiply(A=@n9, B=@n10), B=@n7.Rand))
  Multiply(A=Multiply(A=@n1.RGB, B=Power(Base=@n2.G, Exp=@n3)), B=LinearInterpolate(A=Multiply(A=@n4.RGB, B=@n5), B=@n4.RGB, Alpha=Add(A=@n8, B=@n11)))
  ParticleRandom()
  ScalarParameter() {defaultValue=0.5, parameterName=Param_1}
  TextureSample() {samplerType=SAMPLERTYPE_Normal, texture=&/T_LightSource_PaperLantern_A_N}
 ]
}
