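# Unlit additive material for a spherical airglow effect in the LanternTown environment, giving a soft, warm-tinted haze by default.
# An inverted Fresnel builds a brighter core added to the Airglow tint, and a second power of it forms the soft-edged falloff mask.
# The mask fades at geometry intersections (DepthFade, or manual scene-minus-pixel depth on mobile) and near the camera.
# A time-of-day function scales it, with zero as the daytime default.
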Material /Game/Effects/Fort_Effects/Effects/Environment/LanternTown/Meshes/M_SphericalAirglow
graph {
 properties {bAutomaticallySetUsageInEditor=false, blendMode=BLEND_Additive, shadingModel=MSM_Unlit, usedShadingModels=Unlit}
 nodes [
  n1 = OneMinus(None=Fresnel() {exponent=1})
  n2 = ScalarParameter() {defaultValue=8, parameterName=Core Glow Power}
  Add(
   A=Multiply(A=Power(Base=@n1, Exp=@n2), B=ScalarParameter() {defaultValue=1.5, parameterName=Core Airglow Scale})
   B=VectorParameter() {parameterName=Airglow, defaultValue=<rgba 1 0.3195 0.1688 1>}.RGB
  )
  n3 = Power(Base=@n1, Exp=ScalarParameter() {defaultValue=7, parameterName=Fresnel Power})
  n4 = DepthFade(Opacity=@n3, FadeDistance=ScalarParameter() {defaultValue=19500, parameterName=Depth fade})
  n5 = Subtract(A=SceneDepth().R, B=PixelDepth().R)
  n6 = ScalarParameter() {defaultValue=18500, parameterName=Mobile Depth}
  n7 = Multiply(A=@n3, B=Saturate(None=Divide(A=@n5, B=@n6)))
  n8 = ScalarParameter() {defaultValue=17500, parameterName=Near Camera Fade}
  Multiply(
   A=Multiply(
    A=ShadingPathSwitch(Default=@n4, Mobile=@n7)
    B=MaterialFunctionCall(Distance=@n8) {materialFunction=&Near_Camera_Fade}.Result
   )
   B=MaterialFunctionCall(
    Value_Morning=ScalarParameter() {defaultValue=0.35, parameterName="Value_Morning (S)"}
    Value_Day=ScalarParameter() {parameterName="Value_Day (S)"}
    Value_Evening=ScalarParameter() {defaultValue=0.35, parameterName="Value_Evening (S)"}
    Value_Night=ScalarParameter() {defaultValue=0.35, parameterName="Value_Night (S)"}
   ) {materialFunction=&Scalar_Time_of_Day}.Result
  )
  ScalarParameter() {parameterName=Param_4}
 ]
}
