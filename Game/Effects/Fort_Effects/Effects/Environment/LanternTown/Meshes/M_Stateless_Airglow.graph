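# Unlit additive material for Niagara sprites that draws a soft airglow effect for the LanternTown environment.
# Particle color feeds a shared emissive adjustment function.
# A UV-centered sphere mask, shaped by a power and scaled by an opacity parameter, forms the round falloff, multiplied by a near-camera fade.
# A CameraOffset function shifts the sprite along the view direction from particle position by a parameterized amount.
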Material /Game/Effects/Fort_Effects/Effects/Environment/LanternTown/Meshes/M_Stateless_Airglow
graph {
 properties {
  bAutomaticallySetUsageInEditor=false
  bUsedWithNiagaraSprites=true
  blendMode=BLEND_Additive
  shadingModel=MSM_Unlit
  usedShadingModels=Unlit
 }
 nodes [
  MaterialFunctionCall(Offset Amount=ScalarParameter() {defaultValue=50, parameterName=Param_3}, World Position=ParticlePositionWS()) {materialFunction=/Engine/Functions/Engine_MaterialFunctions02/WorldPositionOffset/CameraOffset}
  MaterialFunctionCall(Emissive=ParticleColor().RGB) {materialFunction=&MF_FX_EmissiveAdjustments}
  n1 = SphereMask(A=TextureCoordinate(), B=Constant() {r=0.5}, Radius=ScalarParameter() {defaultValue=0.5}, Hardness=ScalarParameter() {parameterName=Param_1})
  n2 = ScalarParameter() {defaultValue=3, parameterName=Param_2}
  Multiply(
   A=Multiply(A=Power(Base=@n1, Exp=@n2), B=ScalarParameter() {defaultValue=0.5, parameterName=Airglow Opacity})
   B=MaterialFunctionCall(Distance=ScalarParameter() {defaultValue=150, parameterName=Cam Fade Distance}) {materialFunction=&Near_Camera_Fade}.Result
  )
 ]
}
